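; Rope actor for the HappyGhost hook-gun item that reels in, carrying caught pickups back to the player.
; OnEndReel drives timelines that retract, scale and wiggle the rope, adding bounce on hits, alongside montage and sound.
; Caught pickups attach to an item slot, firing a server quest event.
; Reel completion tosses items, adds ammo for non-pickup catches and destroys the rope, with a client-only duplicate covering unreliable finish delivery.
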(event Custom|AttachNonFishingPickups
  (params (InputPin :type Fort%20Pickup%20Object%20Reference%20%28by%20ref%29))
  (bind _playerpawn (Variables|Default|GetPlayerPawn))
  (Utilities|Array|Add (Variables|Default|GetPickedUpPickups) InputPin)
  (Transformation|AttachActorToComponent InputPin (Variables|Default|GetItemSlot) "None" "SnapToTarget" "KeepRelative" "KeepWorld" true)
  ;[0.05823 0.5 0.4473 0.5] QUESTS: Fire "Caught Pickup" Event
  (Networking|SwitchHasAuthority
    (:Authority
      (Utilities|IsValid _playerpawn
        (:Is%20Valid
          (Utilities|FlowControl|DoOnce false
            (:Completed
              (Ability|SendGameplayEventtoActor _playerpawn "(TagName=\"Event.Quest.CaughtPickup\")" (Utilities|Struct|MakeGameplayEventData "(TagName=\"Event.Quest.CaughtPickup\")" _playerpawn InputPin 0 0 "()" (GameplayTags|GetOwnedGameplayTags _playerpawn :!class /Script/GameplayTags.BlueprintGameplayTagLibrary) (MarkableActor|GetTargetTagsforQuestEvent InputPin))))))))))

(event Custom|DropAndDestroy
  ;[1 1 1] Drop the items picked caught.
  (Utilities|FlowControl|DoOnce false
    (:Completed
      (for _array_element (Actor|GetAttachedActors true)
        (Utilities|IsValid _array_element
          (:Is%20Valid
            (bind _asfort_pickup (Utilities|Casting|CastToFortPickup _array_element))
            (Pickup|TossPickup _asfort_pickup (Transformation|GetActorLocation _asfort_pickup) 0 0 true false "Tossed" "Unset")
            (if (not (Utilities|Array|ContainsItem (Variables|Default|GetPickedUpPickups) _asfort_pickup))
              (bind _asfort_weapon_item_definition (Utilities|Casting|CastToFortWeaponItemDefinition (Class|FortItemFunctionLibrary|BreakFortItemEntry (Pickup|GetItemEntry _asfort_pickup 0))))
              (Fortnite|Weapon|CreateTossAmmoPickupforWeaponItemDefinitionatLocation _asfort_weapon_item_definition (Variables|Default|GetT_FlopperSpawn) (Transformation|GetActorLocation _asfort_pickup))
              (Utilities|FlowControl|DoOnce false
                (:Completed
                  (Audio|SpawnSoundAttached (Variables|Default|GetSound_Pickup) (Variables|Default|GetItemSlot))))))))
      (Utilities|IsValid self
        (:Is%20Valid
          (Actor|SetLifeSpan 0.2))))))

(event Custom|ReelFinished
  ;[1 1 1] When the reel is finish play FX, Hide the rope, and tell the rope to drop the items and destroy
  (Utilities|FlowControl|DoOnce false
    (:Completed
      (Development|PrintString "Reel Finished" true true "(R=1.000000,G=0.126427,B=0.246205,A=1.000000)")
      (Niagara|SpawnSystemAttached "/Game/Athena/Items/Consumables/HappyGhost/FX/NS_HappyGhost_Retract.NS_HappyGhost_Retract" (Variables|Weapon|GetWeaponMesh) "muzzle" "0, 0, 0" "0, 0, 0" "KeepRelativeOffset" true true "AutoRelease")
      (Audio|SpawnSoundAttached (Variables|Default|GetReattachSound) (Variables|/Script/Engine.Actor|GetRootComponent (Variables|Default|GetPlayerPawn)))
      (Rendering|SetVisibility (Variables|Default|GetRopeProjectileMeshVis) false false)
      (Development|SetHiddeninGame (Variables|StaticMesh|GetRopeMesh) true false)
      (Ability|SendGameplayEventtoActor (Variables|Default|GetPlayerPawn) (Variables|Default|GetT_Stop) :!pinned true) ; Plays reattach anim
      (CallFunction|DropandDestroy))))

(event Custom|PlayMontage
  (bind (_ismale _isfemale) (FortPlayerPawn|GetGender (Variables|Default|GetPlayerPawn)))
  (bind _playerpawn (Variables|Default|GetPlayerPawn))
  ;[1 1 1] Play Montage
  (Animation|PlayAnimMontage _playerpawn (select (or (Variables|Default|GetHitPawnOrFish) (Variables|Default|GetHitDamagable)) (Variables|Default|GetReelMontage) (select _isfemale (Variables|Default|GetNoHitMontage_Female) (Variables|Default|GetNoHitMontage))) (Variables|Default|GetMontagePlayRate) "None")
  (if (or (Variables|Default|GetHitPawnOrFish) (Variables|Default|GetHitDamagable))
    (Game|Feedback|LocalPlayForceFeedback (Fortnite|GetFortPlayerControllerfromActor _playerpawn) "/Game/Weapons/ForceFeedback/FF_Athena_HappyGhost_HitReel.FF_Athena_HappyGhost_HitReel" false "None")))

(event Custom|ReelFinishedClientOnly
  ;[1 1 1] When the reel is finish play FX, Hide the rope, and tell the rope to drop the items and destroy
  (Utilities|FlowControl|DoOnce false
    (:Completed
      (Development|PrintString "Reel Finished" true true "(R=1.000000,G=0.126427,B=0.246205,A=1.000000)")
      (Niagara|SpawnSystemAttached "/Game/Athena/Items/Consumables/HappyGhost/FX/NS_HappyGhost_Retract.NS_HappyGhost_Retract" (Variables|Weapon|GetWeaponMesh) "muzzle" "0, 0, 0" "0, 0, 0" "KeepRelativeOffset" true true "AutoRelease")
      (Audio|SpawnSoundAttached (Variables|Default|GetReattachSound) (Variables|/Script/Engine.Actor|GetRootComponent (Variables|Default|GetPlayerPawn)))
      (Rendering|SetVisibility (Variables|Default|GetRopeProjectileMeshVis) false false)
      (Development|SetHiddeninGame (Variables|StaticMesh|GetRopeMesh) true false)
      (Ability|SendGameplayEventtoActor (Variables|Default|GetPlayerPawn) (Variables|Default|GetT_Stop) :!pinned true) ; Plays reattach anim
      (CallFunction|DropandDestroy))))

(event EventBeginPlay
  ;[1 1 1] EVENT -> BeginPlay
  (EventDispatchers|BindEventtoOnEndReel (EventDispatchers|CreateEvent :!event Custom|End%20Reel)))

(event Custom|End%20Reel
  (bind _ropeprojectilemeshvis (Variables|Default|GetRopeProjectileMeshVis))
  (bind _returnvalue (Transformation|GetSocketLocation _ropeprojectilemeshvis "End"))
  (bind _endlocation (Variables|FortHookGunRope|GetEndLocation))
  (bind _bounce (Variables|Default|GetBounce))
  (bind _returnvalue_2 (+ _bounce _endlocation))
  (bind _startlocation (Variables|FortHookGunRope|GetStartLocation))
  (bind _ropemesh (Variables|StaticMesh|GetRopeMesh))
  (bind _ropemesh_2 (Variables|StaticMesh|GetRopeMesh))
  (bind _startlocation_2 (Variables|FortHookGunRope|GetStartLocation))
  ;[1 1 1] Reel In
  (Utilities|FlowControl|DoOnce false
    (:Completed
      (CallFunction|PlayMontage)
      (Utilities|FlowControl|Sequence
        (:then_0
          (Audio|SpawnSoundAttached (Variables|Default|GetReelSound) (Variables|/Script/Engine.Actor|GetRootComponent (Variables|Default|GetPlayerPawn)))
          (Rendering|SetVisibility (Variables|Default|GetRopeProjectileMeshVis) true false))
        (:then_1
          (Components|Timeline|SetPlayRate (Variables|B_Athena_HappyGhost_Wire|GetReel%20In) (Variables|Default|GetReelinPlayRate))
          (bind (_direction _reel_lerp_controller _tighten) (Timeline|Reel%20In PlayFromStart)
            (:Update
              (CallFunction|SetStartPointOfRope)
              (Transformation|SetWorldLocationAndRotation (Variables|StaticMesh|GetRopeProjectileMesh) (Math|Vector|Lerp(Vector) _returnvalue_2 _startlocation _reel_lerp_controller) (Math|Rotator|FindLookatRotation _startlocation _returnvalue_2) false false)
              (bind _distance (Class|MissionBlueprintFunctionLibrary|DistanceBetweenTwoVectors _startlocation_2 (- (Math|Vector|Lerp(Vector) _returnvalue_2 _startlocation _reel_lerp_controller) (- (Math|Vector|Lerp(Vector) _returnvalue_2 _startlocation _reel_lerp_controller) _returnvalue))))
              (Transformation|SetRelativeScale3D _ropemesh (Math|Vector|MakeVector (/ _distance 1e+02) 1.0 1.0))
              (Transformation|SetWorldRotation _ropemesh (Math|Rotator|FindLookatRotation _startlocation_2 (- (Math|Vector|Lerp(Vector) _returnvalue_2 _startlocation _reel_lerp_controller) (- (Math|Vector|Lerp(Vector) _returnvalue_2 _startlocation _reel_lerp_controller) _returnvalue))) false false)
              (Rendering|Material|SetScalarParameterValueonMaterials _ropemesh_2 "ReturnWiggle" (Math|Float|Lerp 0.0 3.0 _tighten))
              (Rendering|Material|SetScalarParameterValueonMaterials _ropemesh_2 "WiggleAmount" (Math|Float|Lerp 0.0 15.0 _tighten))
              (Utilities|IsValid (Actor|GetOwner)
                (:Is%20Not%20Valid
                  (Rendering|SetVisibility (Variables|Default|GetRopeProjectileMeshVis) false false))))
            (:Finished
              ;;;[1 1 1] Sometimes the client is not firing off the event. Try to multicast it and also send it on the client. We are gating the other function behind a do once so that we only play it once on the client. This is just so that it gets to the client faster in the case that we did not lose it.
              ;;[1 0.6907 0] Need to check with engineering why the reel timeline end may not be playing on the client. It is hitting the timeline, just never firing finished.
              (if (Networking|IsServer)
                (CallFunction|ReelFinished)
                (else
                  (CallFunction|ReelFinishedClientOnly))))))
        (:then_2
          (if (or (Variables|Default|GetHitWorld) (Variables|Default|GetHitPawnOrFish))
            (bind (_direction_2 _bounceup _spin) (Timeline|BounceOff PlayFromStart)
              (:Update
                (Variables|Default|SetBounce (* (Variables|Default|GetHitNormal) _bounceup))
                (Transformation|SetRelativeRotation (Variables|Default|GetRopeProjectileMeshVis) (Math|Rotator|Lerp(Rotator) (Math|Rotator|ScaleRotator(Float) "70.000000,25.000000,0.000000" (* _spin (Math|Conversions|ToFloat(Boolean) (Variables|Default|GetHitWorld)))) "0, 0, 0" _reel_lerp_controller true) false false)))))))))

(event Custom|TightenRope
  ;[1 1 1] Tighten Rope
  (bind (_direction _newtrack_0) (Timeline|Tight PlayFromStart)
    (:Update
      (Rendering|Material|SetScalarParameterValueonMaterials (Variables|StaticMesh|GetRopeMesh) "WiggleAmount" (Math|Float|Lerp 0.0 15.0 _newtrack_0)))))

(fn ConstructionScript)
