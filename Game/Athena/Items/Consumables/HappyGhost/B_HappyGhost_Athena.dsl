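; B_HappyGhost_Athena is the weapon Blueprint for the HappyGhost consumable in Athena.
; A replicated HideProj flag's OnRep hides or unhides the "Slide" bone on the weapon mesh, toggling the visible projectile part.
; Equipping and attaching the weapon calls the parent handler and then resets HideProj to false.
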(event FortWeapon|EventOnEquippedAndAttached
  (|CallParentFunction)
  (Variables|Default|SetHideProj false))

(fn OnRep_HideProj
  (bind _weaponmesh (Variables|Visual|GetWeaponMesh))
  (if (Variables|Default|GetHideProj)
    (Components|SkinnedMesh|HideBonebyName _weaponmesh "Slide" "PBO_None")
    (else
      (Components|SkinnedMesh|UnHideBonebyName _weaponmesh "Slide"))))

(fn ConstructionScript
  (|CallParentFunction))
