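; Passive gameplay ability for the HappyGhost consumable weapon, driven by gameplay events for projectile spawn and stop.
; On spawn it applies the blocking active effect, caches pawn and weapon, hides the weapon's projectile and starts a failsafe timer that force-ends the ability.
; On stop it removes the effect, unhides the projectile, and, if ammo is depleted, plays a sound and toss montage before removing the item.
; Ending the ability also removes the effect by tag.
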(event Ability|EventActivateAbilityFromEvent
  (params (EventData :type Gameplay%20Event%20Data%20Structure%20%28by%20ref%29))
  (bind _returnvalue (Ability|GetAvatarActorfromActorInfo))
  (bind _weapon (Variables|Default|GetWeapon))
  (bind _fortpawn (Variables|Default|GetFortPawn))
  ;[1 1 1] Apply the blocking effects while the ability is active. Remove the effect when the ability is called to stop
  (|CallParentFunction EventData)
  (switch Utilities|FlowControl|Switch|SwitchonGameplayTag (Utilities|Struct|BreakGameplayEventData EventData)
    (:Event.Athena.Item.HappyGhost.Stop
      (Ability|RemoveGameplayEffectFromOwnerWithGrantedTags (Variables|Default|GetT_Active))
      ;[1 1 1] UnHide Projectile
      (Variables|/Game/Athena/Items/Consumables/HappyGhost/B_HappyGhost_Athena.B_HappyGhost_Athena_C|SetHideProj false (Variables|Default|GetWeapon))
      ;[1 1 1] Check if out of ammo
      (Utilities|IsValid _weapon
        (:Is%20Valid
          (if (<= (FortWeapon|GetRemainingAmmo _weapon) 0)
            ;[1 1 1] Handles destruction of the weapon. Called when the ammo counts is <1 - Double checks that the ammo is indeed 0, plays the throw away montage and then removes the weapon
            (Audio|PlaySoundatLocation (Variables|Default|GetOutOfChargesSound) (Transformation|GetActorLocation _fortpawn))
            (Animation|Montage|PlayMontage (Variables|/Script/Engine.Character|GetMesh _fortpawn) "/Game/Animation/Game/MainPlayer/Combat/Guns/GrapplingHook/GrapplingHook_GunToss_M.GrapplingHook_GunToss_M") ; gun toss montage
            (Utilities|FlowControl|Delay 0.05)
            (Inventory|RemoveItemFromPlayerByGuid (Fortnite|GetFortPlayerControllerfromActor (Variables|Default|GetFortPawn)) (FortWeapon|GetInventoryGUID (Ability|GetCurrentSourceWeapon self)) 1 true)
            (Ability|EndAbility)
            (else
              _)))
        (:Is%20Not%20Valid
          _))
      (Ability|EndAbility))
    (:Event.Athena.Item.HappyGhost.ProjSpawned
      (Ability|ApplyGameplayEffectToOwner (Variables|Default|GetGE_Active))
      ;[1 1 1] Save Out the Weapon
      (Utilities|IsValid _returnvalue
        (:Is%20Valid
          (bind _asfort_player_pawn (Utilities|Casting|CastToFortPlayerPawn _returnvalue))
          (Variables|Default|SetFortPawn _asfort_player_pawn)
          (bind _asb_happy_ghost_athena (Utilities|Casting|CastToB_HappyGhost_Athena (Variables|/Script/FortniteGame.FortPawn|GetCurrentWeapon _asfort_player_pawn)))
          (bind _output_get (Variables|Default|SetWeapon _asb_happy_ghost_athena))
          (Variables|/Game/Athena/Items/Consumables/HappyGhost/B_HappyGhost_Athena.B_HappyGhost_Athena_C|SetHideProj true _output_get)
          ;[1 1 1] If the ability is still running for some reason just kill it.
          (Utilities|Time|SetTimerbyEvent (delegate Custom|ForceEnd) (Variables|Default|GetFailsafeDelay)))))))

(event Custom|ForceEnd
  ;[1 1 1] If the ability is still running for some reason just kill it.
  (Variables|/Game/Athena/Items/Consumables/HappyGhost/B_HappyGhost_Athena.B_HappyGhost_Athena_C|SetHideProj false (Variables|Default|GetWeapon))
  (Ability|EndAbility))

(event Ability|EventOnEndAbility
  (params (bWasCancelled :type Boolean))
  (Ability|RemoveGameplayEffectFromOwnerWithGrantedTags (Variables|Default|GetT_Active)))
