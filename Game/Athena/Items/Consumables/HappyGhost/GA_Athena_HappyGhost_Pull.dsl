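; Happy Ghost pull ability that yanks its avatar toward the instigator of the granting effect.
; Player targets, including owners of melee-block actors, pass friendly-fire, vehicle and skydive checks, then leave skydiving or ziplines before an authority-side launch.
; AI pawns without ignore tags are forced into falling before launching, so nav-mesh walking cannot block it.
; CalculateArc solves a custom-arc velocity landing a tuned distance from the launcher, toward the target.
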(event Ability|EventActivateAbility
  (bind _returnvalue (Ability|GetAvatarActorfromActorInfo))
  (bind _returnvalue_2 (Ability|GetGrantedbyEffectContext))
  (bind _returnvalue_3 (Ability|EffectContext|GetEffectCauser _returnvalue_2))
  (bind _instigator (Variables|/Script/Engine.Actor|GetInstigator _returnvalue_3))
  (bind _target (Variables|Default|GetTarget))
  (bind _target_2 (Variables|Default|GetTarget))
  ;[1 1 1] Save Player player out, kick from vehicle, stop gliding.
  (if (Math|Conversions|ToBoolean(Integer) (Math|Float|Floor (ScalableFloat|GetValueatLevel (Variables|Default|GetRow_PullEnabled))))
    (Development|PrintString "Pull")
    (Utilities|IsValid _instigator
      (:Is%20Valid
        (bind _asfort_player_pawn (Utilities|Casting|CastToFortPlayerPawn _instigator)
          (:then
            (Variables|Default|SetLauncher _asfort_player_pawn)
            (bind _returnvalue_4 (GameplayTags|HasMatchingGameplayTag(Message) _returnvalue "(TagName=\"Gameplay.Melee.Block\")")) ; Pull owner of a melee blocking collision actor
            (bind _asfort_player_pawn_2 (Utilities|Casting|CastToFortPlayerPawn (select _returnvalue_4 (Actor|GetOwner _returnvalue) _returnvalue))
              (:then
                (Variables|Default|SetTarget _asfort_player_pawn_2)
                (if (Math|Conversions|ToBoolean(Integer) (Math|Float|Floor (ScalableFloat|GetValueatLevel (Variables|Default|GetRow_CanPullFriends))))
                  _
                  (elif (Team|OnSameTeam (Variables|Default|GetLauncher) (Variables|Default|GetTarget))
                    (Ability|EndAbility)))
                ;[1 1 1] If the target is an ally do not pull if they are skydiving or in a vehicle
                (if (Team|OnSameTeam (Variables|Default|GetLauncher) _target_2)
                  (if (or (Vehicle|IsinVehicle _target_2) (Parachute|IsSkydiving _target_2))
                    (Ability|EndAbility)
                    (else
                      _))
                  (else
                    _))
                (Tags|RemoveNonReplicatedTagtoPawn _target (Variables|Default|GetT_InVehicle))
                (Parachute|EndSkydiving _target)
                (Ability|HasAuthority
                  (:True
                    (bind _outlaunchvelocity (CallFunction|CalculateArc (Variables|Default|GetTarget)))
                    ;[1 1 1] Launch Players
                    (bind _launchedpawn (Utilities|LaunchPlayerPawn (Variables|Default|GetLauncher) (Variables|Default|GetTarget) _outlaunchvelocity true true false true true true true))
                    (Pawn|Character|ResetFallingHeight _launchedpawn)
                    (Utilities|FlowControl|Delay 0.1)
                    (Ability|EndAbility))
                  (:False
                    (bind _returnvalue_5 (Zipline|IsZiplining _target))
                    (if _returnvalue_5
                      (Zipline|EndZiplining _target false false)))))
              (:CastFailed
                (bind _asfort_aipawn (Utilities|Casting|CastToFortAIPawn (Ability|GetAvatarActorfromActorInfo))
                  (:then
                    (Variables|Default|SetAITarget _asfort_aipawn)
                    ;[1 1 1] Don't Launch if these tags are on the AI
                    (bind _aitarget (Variables|Default|GetAITarget)
                      (:then
                        (if (GameplayTags|HasAnyMatchingGameplayTags _aitarget (Variables|Default|GetTC_TagsForAIToIgnoreLaunch))
                          (Ability|EndAbility)
                          (else
                            ;[1 1 1] Launc AI Pawns (Like Boars)
                            (bind _aitarget_2 (Variables|Default|GetAITarget)
                              (:then
                                (bind _outlaunchvelocity_2 (CallFunction|CalculateArc _aitarget_2))
                                (Pawn|Components|CharacterMovement|SetMovementMode (Variables|/Script/Engine.Character|GetCharacterMovement _aitarget_2) "MOVE_Falling" 0) ; Set Falling in case we were Nav Mesh Walking.\nThat was stopping the Launch from working.
                                ;;[1 1 1] Launch Char
                                (Character|LaunchCharacter _aitarget_2 _outlaunchvelocity_2 true true)
                                (Pawn|Character|ResetFallingHeight _aitarget_2)
                                (Utilities|FlowControl|Delay 0.1)
                                (Ability|EndAbility))
                              (:else)))))
                      (:else)))
                  (:CastFailed
                    (Ability|EndAbility))))))
          (:CastFailed
            _)))
      (:Is%20Not%20Valid
        _))
    (Ability|EndAbility)
    (else
      (Ability|EndAbility))))

(fn CalculateArc
  (params (ArcTarget :type Actor%20Object%20Reference))
  ;[1 1 1] Calc Velocity
  (bind _outlaunchvelocity (Game|SuggestProjectileVelocityCustomArc (Transformation|GetActorLocation ArcTarget) (+ (Transformation|GetActorLocation (Variables|Default|GetLauncher)) (* (Math|Vector|Normalize (- (Transformation|GetActorLocation ArcTarget) (Transformation|GetActorLocation (Variables|Default|GetLauncher)))) (ScalableFloat|GetValueatLevel (Variables|Default|GetRow_MaxPullLocation)))) (Variables|Default|GetLaunchGrav) (Variables|Default|GetLaunchArc))) ; Finds out how much force is needed to put the target 300 units in front of where instigator is facing
  (return _outlaunchvelocity))
