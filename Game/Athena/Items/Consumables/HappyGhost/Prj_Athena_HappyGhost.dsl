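; Happy Ghost hook-gun projectile that catches pawns, fish and pickups on a server-spawned wire, which it tells when to end reeling.
; Hits set wire state and reel-in rate, send quest events, and knock players out of parachutes and ziplines.
; Overlaps spawn hooked items from fishing spots or attach qualifying pickups to the wire.
; Ammo is consumed server-side on certain catches, while water impact and fishing effects play cosmetically.
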(event EventBeginPlay
  ;[1 1 1] Spawn and attach wire
  (bind _asfort_player_pawn (Utilities|Casting|CastToFortPlayerPawn (Game|GetInstigator)))
  (bind _output_get (Variables|Default|SetFortPawn _asfort_player_pawn))
  (Ability|SendGameplayEventtoActor _output_get (Variables|Default|GetEvent_ProjSpawned))
  (bind _asb_happy_ghost_athena (Utilities|Casting|CastToB_HappyGhost_Athena (Actor|GetOwner)))
  (Variables|Default|SetWeaponActor _asb_happy_ghost_athena)
  (if (Networking|IsServer)
    (bind _returnvalue (Game|SpawnActorfromClass "/Game/Athena/Items/Consumables/HappyGhost/B_Athena_HappyGhost_Wire.B_Athena_HappyGhost_Wire_C" (Math|Transform|MakeTransform (Transformation|GetActorLocation)) "Undefined" "MultiplyWithRoot" (Variables|Default|GetWeaponActor) (Variables|Default|GetFortPawn) 0))
    (bind _output_get_2 (Variables|Default|SetRope _returnvalue))
    (Class|FortHookGunRope|SetWeaponandProjectile _output_get_2 (Variables|Default|GetWeaponActor) self)))

(event EventDestroyed
  (bind _rope (Variables|Default|GetRope))
  ;[1 1 1] Handle Destroy
  (Utilities|IsValid _rope
    (:Is%20Valid
      (Variables|/Script/FortniteGame.FortHookGunRope|SetbProjectileGone true _rope)
      (Development|PrintString "Send End Reel"))))

(event Collision|EventHit
  (params (MyComp :type Primitive%20Component%20Object%20Reference) (Other :type Actor%20Object%20Reference) (OtherComp :type Primitive%20Component%20Object%20Reference) (bSelfMoved :type Boolean) (HitLocation :type Vector :default "0, 0, 0") (HitNormal :type Vector :default "0, 0, 0") (NormalImpulse :type Vector :default "0, 0, 0") (Hit :type Hit%20Result%20Structure%20%28by%20ref%29))
  (bind (_aspawn _bsuccess) (Utilities|Casting|CastToPawn Other))
  (bind (_asbuilding_actor _bsuccess_2) (Utilities|Casting|CastToBuildingActor Other))
  (bind _rope (Variables|Default|GetRope))
  ;[1 1 1] Handle Hits
  (Collision|SetActorEnableCollision false)
  (Utilities|IsValid _rope
    (:Is%20Valid
      (CallFunction|SendQuestEvent Other)
      (Variables|/Game/Athena/Items/Consumables/HappyGhost/B_Athena_HappyGhost_Wire.B_Athena_HappyGhost_Wire_C|SetHitPawnOrFish _bsuccess _rope)
      (Variables|/Game/Athena/Items/Consumables/HappyGhost/B_Athena_HappyGhost_Wire.B_Athena_HappyGhost_Wire_C|SetHitWorld (not _bsuccess) _rope)
      (Variables|/Game/Athena/Items/Consumables/HappyGhost/B_Athena_HappyGhost_Wire.B_Athena_HappyGhost_Wire_C|SetHitNormal HitNormal _rope)
      (Utilities|IsValid _aspawn
        (:Is%20Valid
          (Utilities|FlowControl|Sequence
            (:then_0
              (if (not (and (Variables|/Script/Engine.Actor|GetbCanBeDamaged _aspawn) _bsuccess))
                (if (not (and (Variables|/Script/Engine.Actor|GetbCanBeDamaged _asbuilding_actor) _bsuccess_2))
                  (bind _returnvalue (GameplayTags|HasAnyMatchingGameplayTags(Message) Other (Variables|Default|GetTC_Vehicle)))
                  (if _returnvalue
                    (if (Networking|IsServer)
                      (CallFunction|HandleCost)
                      (else
                        _))
                    (Variables|/Game/Athena/Items/Consumables/HappyGhost/B_Athena_HappyGhost_Wire.B_Athena_HappyGhost_Wire_C|SetHitDamagable true (Variables|Default|GetRope))
                    (Variables|Default|SetReelInPlayRate (Variables|Default|GetHitReelInRate))
                    (else
                      _))
                  (Variables|/Game/Athena/Items/Consumables/HappyGhost/B_Athena_HappyGhost_Wire.B_Athena_HappyGhost_Wire_C|SetReelinPlayRate (Variables|Default|GetReelInPlayRate) (Variables|Default|GetRope))
                  (Actor|SetLifeSpan 0.05))))
            (:then_1
              ;[1 1 1] Removes Hit Players from parachuting
              (bind _asfort_player_pawn (Utilities|Casting|CastToFortPlayerPawn _aspawn))
              (if (Parachute|IsParachuteOpen _asfort_player_pawn)
                (Parachute|CloseParachuteAndFall _asfort_player_pawn)))
            (:then_2
              ;[1 1 1] Remove hit players from ziplines and ascenders
              (bind _returnvalue_2 (GameplayTags|HasMatchingGameplayTag(Message) _aspawn "(TagName=\"Abilities.Generic.Athena.ZipLine\")"))
              (if _returnvalue_2
                (Ability|SendGameplayEventtoActor _aspawn "(TagName=\"Abilities.Generic.Athena.ZipLine.Exit\")")))))
        (:Is%20Not%20Valid
          (Utilities|IsValid _asbuilding_actor
            (:Is%20Valid
              _)
            (:Is%20Not%20Valid
              _)))))))

(event Collision|EventActorBeginOverlap
  (params (OtherActor :type Actor%20Object%20Reference))
  (bind _returnvalue (Transformation|GetActorLocation))
  (bind _rope (Variables|Default|GetRope))
  ;[1 1 1] Handle Hitting World Flopper Spawn
  (bind _asbga_athena_flopper_spawn_world (Utilities|Casting|CastToBGA_Athena_FlopperSpawn_World OtherActor)
    (:then
      (Variables|Default|SetReelInPlayRate (Variables|Default|GetHitFishReelInRate))
      (CallFunction|SendQuestEvent _asbga_athena_flopper_spawn_world)
      (Utilities|IsValid _rope
        (:Is%20Valid
          (Fortnite|Fishing|FireEventBeginFishingSession (Variables|Default|GetFortPawn))
          (Class|BGAAthenaFlopperSpawnParent|SpawnItem :self _asbga_athena_flopper_spawn_world :Target (Game|GetInstigator self) :Hooked true :HookComponent (Variables|/Game/Athena/Items/Consumables/HappyGhost/B_Athena_HappyGhost_Wire.B_Athena_HappyGhost_Wire_C|GetItemSlot _rope) :SpawnLoc _returnvalue :Player (Variables|Default|GetFortPawn) :ItemSpawnDelay (Variables|Default|GetItemSpawnDelay) :CaughtWithFishingRod false :CaughtWithHappyGhost true :ItemUsedToFish (Variables|/Script/FortniteGame.FortPawn|GetCurrentWeapon (Variables|Default|GetFortPawn)))
          (Audio|PlaySoundatLocation (Variables|Default|GetHit%20Fish%20Sound) _returnvalue)
          (CallFunction|SpawnItemsFX)
          (if (Networking|IsServer)
            (CallFunction|HandleCost)
            (else
              _))
          (Variables|/Game/Athena/Items/Consumables/HappyGhost/B_Athena_HappyGhost_Wire.B_Athena_HappyGhost_Wire_C|SetHitPawnOrFish true (Variables|Default|GetRope))
          ;[1 1 1] Handle Hits
          (Variables|/Game/Athena/Items/Consumables/HappyGhost/B_Athena_HappyGhost_Wire.B_Athena_HappyGhost_Wire_C|SetReelinPlayRate (Variables|Default|GetReelInPlayRate) (Variables|Default|GetRope))
          (Actor|SetLifeSpan 0.05))))
    (:CastFailed
      (bind _asfort_pickup (Utilities|Casting|CastToFortPickup OtherActor))
      (CallFunction|CustomPickupAttach _asfort_pickup))))

(event Custom|HandleCost
  (bind _weaponactor (Variables|Default|GetWeaponActor))
  ;[1 1 1] Handle Cost
  (if (FortWeapon|Ammo|HasInfiniteUses _weaponactor)
    (FortWeapon|Ammo|ConsumeAmmo _weaponactor 0 false)
    (else
      (FortWeapon|Ammo|ConsumeAmmo _weaponactor 1 false))))

(event Custom|SpawnItemsFX
  ;[1 1 1] Play Spawn Fx and SFX for fishing pools
  (Utilities|FlowControl|DoOnce false
    (:Completed
      (if (not (Networking|IsServer))
        (bind _outhit (Collision|LineTraceByChannel (+ (Transformation|GetActorLocation) (Variables|Default|GetSpawnItemsTraceOffset)) (- (Transformation|GetActorLocation) (Variables|Default|GetSpawnItemsTraceOffset)) "TraceTypeQuery12"))
        (bind _asfort_water_body_actor (Utilities|Casting|CastToFortWaterBodyActor (.HitActor (Collision|BreakHitResult _outhit))))
        (bind _outwatersurfaceinfo (WaterBody|GetWaterSurfaceInfoStruct _asfort_water_body_actor (Transformation|GetActorLocation) true false false))
        (bind _returnvalue (Effects|SpawnEmitteratLocation :EmitterTemplate (Variables|Default|GetFX_SpawnItems) :Location (Utilities|Struct|BreakWaterSurfaceInfo _outwatersurfaceinfo) :Scale "1.000000,1.000000,1.000000" :bAutoDestroy true :PoolingMethod "AutoRelease" :bAutoActivateSystem true))
        (Effects|Components|ParticleSystem|SetFloatParameter _returnvalue "WaterBodyIndex" (Math|Conversions|ToFloat(Integer) (.WaterBodyIdx (Utilities|Struct|BreakWaterSurfaceInfo _outwatersurfaceinfo))))
        (Effects|Components|ParticleSystem|SetFloatParameter _returnvalue "WaterDepth" (.WaterDepth (Utilities|Struct|BreakWaterSurfaceInfo _outwatersurfaceinfo)))
        (Audio|PlaySoundatLocation (Variables|Default|GetSound_SpawnItems) (Transformation|GetActorLocation))))))

(event Custom|CustomPickupAttach
  (params (FortPickup :type Fort%20Pickup%20Object%20Reference))
  ;[1 1 1] Add the custom pickup to the wire.
  (if (Networking|IsServer)
    (if (Math|Boolean|ANDBoolean (Utilities|NotEqual(Object) (Actor|GetAttachParentActor FortPickup) self) (<= (Transformation|GetSquaredDistanceTo FortPickup self) (Variables|Default|GetItemAttachMinDist)) (>= (Transformation|GetSquaredDistanceTo (Game|GetInstigator) FortPickup) (Variables|Default|GetItemAttachDistanceRequirement)))
      (Class|BAthenaHappyGhostWire|AttachNonFishingPickups (Variables|Default|GetRope) FortPickup)
      (CallFunction|SendQuestEvent FortPickup))))

(event EventOnTouched
  (params (OtherActor :type Actor%20Object%20Reference) (OtherComp :type Primitive%20Component%20Object%20Reference) (HitResult :type Hit%20Result%20Structure%20%28by%20ref%29) (bIsOverlap :type Boolean))
  (bind _waterhit (Variables|Default|GetWaterHit))
  (bind (_bblockinghit _binitialoverlap _time _distance _location _impactpoint _normal _impactnormal _physmat) (Collision|BreakHitResult _waterhit))
  (bind _waterinteractioncomponent (Variables|Water|GetWaterInteractionComponent))
  (bind (_waterplanelocation _waterplanenormal _watersurfaceposition _waterdepth _waterbodyidx) (FX|GetLastWaterSurfaceInfo _waterinteractioncomponent :!class /Script/FortniteGame.FortWaterInteractionComponent))
  ;[0 0.298 1] Water Impact FX
  (if bIsOverlap
    (Variables|Default|SetWaterHit HitResult)
    (Networking|CanExecuteCosmeticEvents
      (:True
        (bind _output_get (Variables|Default|SetWaterHitMat _physmat))
        (Utilities|IsValid _output_get
          (:Is%20Valid
            (if (Utilities|Enum|Equal(Enum) (Variables|/Script/PhysicsCore.PhysicalMaterial|GetSurfaceType _output_get) "SurfaceType8")
              (Audio|PlaySoundatLocation "/Game/Sounds/Fort_Weapons/Happy_Ghost/SC_Water_Bullet_Impact.SC_Water_Bullet_Impact" _location)
              (Utilities|FlowControl|Delay 0.01)
              (bind _returnvalue (Effects|SpawnEmitteratLocation (Variables|Default|GetWaterImpactFX) _location (Math|Conversions|RotationFromXVector _waterplanenormal)))
              (Effects|Components|ParticleSystem|SetFloatParameter _returnvalue "WaterDepth" _waterdepth)
              (Effects|Components|ParticleSystem|SetFloatParameter _returnvalue "WaterBodyIndex" (Math|Conversions|ToFloat(Integer) _waterbodyidx)))))))))

(fn SendQuestEvent
  (params (Hit%20Actor :type Actor%20Object%20Reference))
  (Ability|SendGameplayEventtoActor (Variables|Default|GetFortPawn) "(TagName=\"Event.Athena.Item.HappyGhost.Hit\")" (Utilities|Struct|MakeGameplayEventData :Instigator (Variables|Default|GetFortPawn) :Target Hit%20Actor :OptionalObject self :ContextHandle "()" :InstigatorTags "(GameplayTags=)" :TargetTags "(GameplayTags=)" :EventMagnitude 0.0 :TargetData "()")))

(fn ConstructionScript
  (|CallParentFunction))
